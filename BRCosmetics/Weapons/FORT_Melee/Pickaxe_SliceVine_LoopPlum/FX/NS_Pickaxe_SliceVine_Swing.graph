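# Niagara swing effect for the SliceVine pickaxe: one Claw_Trails emitter draws four short-lived anim-trail ribbons between skeletal mesh bones or sockets.
# Ribbons narrow and fade with age, also scaled by a system TrailFade value.
# User.Variant switches age-based color curves, alpha strength and a secondary color; RGB also follows a time-of-day multiplier.
# Two ribbon renderers draw the trails, with a 1.5x secondary width and a time-driven U override.
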NiagaraSystem /BRCosmetics/Weapons/FORT_Melee/Pickaxe_SliceVine_LoopPlum/FX/NS_Pickaxe_SliceVine_Swing
dependencies {
 usedDataInterfaces [&NiagaraDataInterfaceCurve, &NiagaraDataInterfaceColorCurve, &NiagaraDataInterfaceSkeletalMesh]
 usedDynamicInputs [&self:SectionIndexFloatSelector_01]
 usedModules [
  &self:SystemSpawnScript.SystemScriptSource.SystemScriptGraph.NiagaraNodeAssignment_0.SetVariables_#1
  &SystemState
  &self:SystemSpawnScript.SystemScriptSource.SystemScriptGraph.NiagaraNodeAssignment_6.SetVariables_#2
  &self:SystemSpawnScript.SystemScriptSource.SystemScriptGraph.NiagaraNodeAssignment_2.SetVariables_#3
 ]
}
emitters {Claw_Trails {
  emitterUpdateScript [
   &EmitterState()
   &FN_SpawnAnimTrail_v2(AnimTrailWidthMode=::FromFirst, Number of Anim Trails=4, Spawn Interval=0)
  ]
  particleSpawnScript [
   /Game/Effects/Niagara/Modules/AnimTrail/FN_InitializeAnimTrail/FN_InitializeAnimTrailBoneIndecies()
   &FN_InitializeAnimTrailBoneIndecies(AnimTrail Selection=1, First Bone or Socket Index=1, Second Bone or Socket Index=1) {module=FN_InitializeAnimTrailBoneIndecies001}
   &FN_InitializeAnimTrailBoneIndecies(AnimTrail Selection=2, First Bone or Socket Index=2, Second Bone or Socket Index=2) {module=FN_InitializeAnimTrailBoneIndecies002}
   &FN_InitializeAnimTrailBoneIndecies(AnimTrail Selection=3, First Bone or Socket Index=3, Second Bone or Socket Index=3) {module=FN_InitializeAnimTrailBoneIndecies003}
   &FN_InitializeAnimTrail(AnimTrailWidth=2.5, Color=<rgba 0.3351 0.3351 0.3351 1>, Lifetime=0.13, SkeletalMesh=$System.Skeletal Mesh) {module=FN_InitializeAnimTrail001}
  ]
  particleUpdateScript [
   &FN_UpdateAnimTrail(SkeletalMesh=$System.Skeletal Mesh)
   &CameraOffset(Camera Offset Amount=10)
   &ScaleRibbonWidth(Ribbon Width Scale=&FloatFromCurve(FloatCurve=DataInterface() {propertyValues {Curve {keys [
         {arriveTangent=-2.083, interpMode=RCIM_Cubic, leaveTangent=-2.083, tangentMode=RCTM_User, time=0.25, value=1}
         {tangentMode=RCTM_User, time=1, value=0.25}
        ]}}}))
   &ScaleColor(
    Scale Alpha=&Multiply_Float(
     A=&FloatFromCurve(
      FloatCurve=DataInterface() {propertyValues {Curve {keys [{arriveTangent=-0.5419, time=0, value=0.25}, <at 1 0>]}}}
      Scale Curve=$System.TrailFade
     )
     B=&MergeHelper_FloatSwitch(0=0.6, 1=0.6, 2=?, 3=?, Index=&MakeIntFromFloat(Float=$User.Variant), Num=::NewEnumerator0)
    )
    Scale RGB=&Multiply_VectorByFloat(
     Vector=&Multiply_Vector(
      A=&MakeVectorFromLinearColorRGB(Color=&MergeHelper_ColorSwitch(
        0=&ColorFromCurve(ColorCurve=DataInterface() {propertyValues {
           AlphaCurve {keys [<at 0 1 RCIM_Cubic>,<at 1 1 RCIM_Cubic>]}
           BlueCurve {keys [<at 0.0021 3.8>, {arriveTangent=-5.1274, interpMode=RCIM_Cubic, leaveTangent=-5.1274, time=0.3468, value=2}, <at 0.5871 0.8>, <at 1 0.5>]}
           GreenCurve {keys [
             <at 0.0021 4>
             {arriveTangent=-6.1484, interpMode=RCIM_Cubic, leaveTangent=-6.1484, time=0.3468, value=1.5667}
             <at 0.5871 0.4026>
             <at 1 0.0083>
            ]}
           RedCurve {keys [<at 0.0021 0>, {arriveTangent=0.3971, interpMode=RCIM_Cubic, leaveTangent=0.3971, time=0.3468, value=0}, <at 0.5871 0.2324>, <at 1 0>]}
          }})
        1=&ColorFromCurve(ColorCurve=DataInterface() {propertyValues {
           AlphaCurve {keys [<at 0 1 RCIM_Cubic>,<at 1 1 RCIM_Cubic>]}
           BlueCurve {keys [
             <at 0.0021 0.1789>
             {arriveTangent=-0.2238, interpMode=RCIM_Cubic, leaveTangent=-0.2238, time=0.3468, value=0.1282}
             <at 0.5871 0.048>
             <at 1 0.0149>
            ]}
           GreenCurve {keys [<at 0.0021 4>, {arriveTangent=-5.1274, interpMode=RCIM_Cubic, leaveTangent=-5.1274, time=0.3468, value=2}, <at 0.5871 1>, <at 1 0.2917>]}
           RedCurve {keys [
             <at 0.0021 0.9154>
             {arriveTangent=-1.1528, interpMode=RCIM_Cubic, leaveTangent=-1.1528, time=0.3468, value=0.5394}
             <at 0.5871 0.2409>
             <at 1 0.1125>
            ]}
          }})
        2=&ColorFromCurve(ColorCurve=DataInterface() {propertyValues {
           AlphaCurve {keys [<at 0 1 RCIM_Cubic>,<at 1 1 RCIM_Cubic>]}
           BlueCurve {keys [<at 0.0021 0>,<at 0.0788 0>,<at 0.3771 0>,<at 1 0>]}
           GreenCurve {keys [<at 0.0021 7.122>,<at 0.0788 1.2008>,<at 0.3771 0.2402>,<at 1 0.0869>]}
           RedCurve {keys [<at 0.0021 10>,<at 0.0788 5>,<at 0.3771 1>,<at 1 0.5>]}
          }})
        3=&ColorFromCurve(ColorCurve=DataInterface() {propertyValues {
           AlphaCurve {keys [<at 0 1 RCIM_Cubic>,<at 1 1 RCIM_Cubic>]}
           BlueCurve {keys [<at 0.0021 10>, {arriveTangent=-15.7241, interpMode=RCIM_Cubic, leaveTangent=-15.7241, time=0.3468, value=2}, <at 0.5871 0.8>, <at 1 0.5>]}
           GreenCurve {keys [
             <at 0.0021 9.715>
             {arriveTangent=-16.1777, interpMode=RCIM_Cubic, leaveTangent=-16.1777, time=0.3468, value=0.9874}
             <at 0.5871 0.2496>
             <at 1 0.0383>
            ]}
           RedCurve {keys [<at 0.0021 0>,<at 0.3468 0 RCIM_Cubic>,<at 0.5871 0>,<at 1 0>]}
          }})
        4=&ColorFromCurve(ColorCurve=DataInterface() {propertyValues {
           AlphaCurve {keys [<at 0 1 RCIM_Cubic>,<at 1 1 RCIM_Cubic>]}
           BlueCurve {keys [<at 0.0021 8.9577>,<at 0.151 1>,<at 0.2954 1>,<at 0.5158 0.5>,<at 1 0.5>]}
           GreenCurve {keys [<at 0.0021 0>,<at 0.151 0.2514>,<at 0.2954 0.5566>,<at 0.5158 0.1561>,<at 1 0.4141>]}
           RedCurve {keys [<at 0.0021 10>,<at 0.151 0.3611>,<at 0.2954 0.3073>,<at 0.5158 0>,<at 1 0>]}
          }})
        5=&ColorFromCurve(ColorCurve=DataInterface() {propertyValues {
           AlphaCurve {keys [<at 0 1 RCIM_Cubic>,<at 1 1 RCIM_Cubic>]}
           BlueCurve {keys [<at 0.0021 10>, <at 0.0997 1>, {interpMode=RCIM_Cubic, leaveTangent=-10.0206, time=0.4588, value=1}]}
           GreenCurve {keys [<at 0.0021 6.108>, <at 0.0997 0.3021>, {interpMode=RCIM_Cubic, leaveTangent=-5.6195, time=0.4588, value=0.3664}]}
           RedCurve {keys [<at 0.0021 9.061>, <at 0.0997 0.7648>, {interpMode=RCIM_Cubic, leaveTangent=-8.7782, time=0.4588, value=0.3938}]}
          }})
        Index=&MakeIntFromFloat(Float=$User.Variant)
        Num=::NewEnumerator0
       ))
      B=&MakeVectorFromLinearColorRGB(Color=$System.TODColorMultiplier)
     )
     Float=&FNTimeOfDayScalar(Night Scale=1, Morning Scale=1, Evening Scale=1)
    )
    ScaleA=true
    ScaleRGB=true
   ) {module=ScaleColor001}
   /SetVariables(
    Particles.SecondaryColor:LinearColor=&MultiplyLinearColor(
     Linear Color=&MergeHelper_ColorSwitch(
      0=<rgba 0.5833 0 0.5326 0.09>
      1=<rgba 0.0382 0.0612 0.1812 0.45>
      2=?
      3=?
      Index=&MakeIntFromFloat(Float=$User.Variant)
      Num=::NewEnumerator0
     )
     Scale Factor=&MakeFloat4FromVectorAndFloat("Vector (RGB)"=<1 1 1>, "Float (Alpha)"=$System.TrailFade)
    )
    Particles.SecondaryWidth:NiagaraFloat=&Multiply_Float(A=$Particles.RibbonWidth, B=1.5)
   ) {isSetParameters=true}
   /SetVariables(Particles.RibbonU0Override=&Multiply_Float(A=$Engine.Time, B=4)) {isSetParameters=true}
  ]
  rendererClasses [&NiagaraRibbonRendererProperties]
  renderers [
   {rendererClass=&NiagaraRibbonRendererProperties}
   {rendererClass=&NiagaraRibbonRendererProperties, rendererIndex=1}
  ]
 }}
summary {
 systemName=NS_Pickaxe_SliceVine_Swing
 emitters [{emitterName=Claw_Trails, rendererClasses [&NiagaraRibbonRendererProperties]}]
 userVariables [
  User.RibbonIDOffset {type=NiagaraInt32}
  User.SwingSectionIndex {type=NiagaraInt32}
  User.Variant {type=NiagaraFloat, defaultValue=0, description=""}
 ]
}
